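# Physics asset for the LucidAzalea medium female player body, built around a dynamic skirt over capsule bodies for the pelvis, thighs and calves.
# Six skirt chains (front, side and back on each side) start at tiny constr anchor capsules linked to the pelvis, with three segments on the left and two on the right.
# Skirt segments have narrow swing and twist limits, while anchor and leg constraints leave rotation free.
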PhysicsAsset /Game/Characters/Player/Female/Medium/Bodies/F_MED_LucidAzalea/Meshes/F_MED_LucidAzalea_Physics
bodies {
 calf_l [{shapeType=Capsule, length=31.7648, radius=6.4632, shapeName=calf_l_capsule, center=<-19.0484 -2.2587 0>, rotation=<rot 90 80.7734 78.0025>}]
 calf_r [{shapeType=Capsule, length=31.7648, radius=6.4632, shapeName=calf_l_capsule, center=<19.0484 2.2587 0>, rotation=<rot 90 80.7734 78.0025>}]
 dyn_skirt_back_l_01 [{shapeType=Capsule, length=17.6784, radius=2.5, shapeName=dyn_skirt_back_l_01_capsule, center=<3.6752 -1.8926 0>, rotation=<rot 0 0 15.4106>}]
 dyn_skirt_back_l_02 [{shapeType=Capsule, length=18.1784, radius=3.5, shapeName=dyn_skirt_back_l_02_capsule, center=<3.6752 -1.4014 0>, rotation=<rot 0 0 15.4106>}]
 dyn_skirt_back_l_03 [{shapeType=Capsule, length=18.1784, radius=3.5, shapeName=dyn_skirt_back_l_03_capsule, center=<3.6752 -1.4014 0>, rotation=<rot 0 0 15.4106>}]
 dyn_skirt_back_l_constr [{shapeType=Capsule, length=0.505, radius=0.505, shapeName=dyn_skirt_back_l_constr_capsule, rotation=<rot 0 -90 0>}]
 dyn_skirt_back_r_01 [{shapeType=Capsule, length=18.0595, radius=2.5, shapeName=dyn_skirt_back_r_01_capsule, center=<3.6752 -1.8926 0>, rotation=<rot 0 0 -14.8409>}]
 dyn_skirt_back_r_02 [{shapeType=Capsule, length=18.0595, radius=3.5, shapeName=dyn_skirt_back_r_02_capsule, center=<3.6752 -1.4014 0>, rotation=<rot 0 0 -14.8409>}]
 dyn_skirt_back_r_constr [{shapeType=Capsule, length=0.505, radius=0.505, shapeName=dyn_skirt_back_r_constr_capsule, rotation=<rot 0 -90 0>}]
 dyn_skirt_front_l_01 [{
   shapeType=Capsule
   length=14.302
   radius=2
   shapeName=dyn_skirt_front_l_01_capsule
   center=<3.6752 -1.7705 0>
   rotation=<rot 0 -0.0001 -12.9064>
  }]
 dyn_skirt_front_l_02 [{
   shapeType=Capsule
   length=14.302
   radius=3
   shapeName=dyn_skirt_front_l_02_capsule
   center=<3.6752 -1.2376 0>
   rotation=<rot 0 -0.0001 -12.9064>
  }]
 dyn_skirt_front_l_03 [{
   shapeType=Capsule
   length=14.302
   radius=3
   shapeName=dyn_skirt_front_l_03_capsule
   center=<3.6752 -0.6934 0>
   rotation=<rot 0 -0.0001 -12.9064>
  }]
 dyn_skirt_front_l_constr [{shapeType=Capsule, length=0.505, radius=0.505, shapeName=dyn_skirt_front_l_constr_capsule, rotation=<rot 0 -90 0>}]
 dyn_skirt_front_r_01 [{
   shapeType=Capsule
   length=16.34
   radius=2
   shapeName=dyn_skirt_front_r_01_capsule
   center=<4.4284 -1.1361 0>
   rotation=<rot 0 -0.0001 15.2577>
  }]
 dyn_skirt_front_r_02 [{
   shapeType=Capsule
   length=17.0051
   radius=2.335
   shapeName=dyn_skirt_front_r_02_capsule
   center=<7.4471 -1.2255 0.0642>
   rotation=<rot 6.7775 1.8576 15.3677>
  }]
 dyn_skirt_front_r_constr [{shapeType=Capsule, length=0.505, radius=0.505, shapeName=dyn_skirt_front_r_constr_capsule, rotation=<rot 0 -90 0>}]
 dyn_skirt_side_l_01 [{shapeType=Capsule, length=13.2998, radius=2, shapeName=dyn_skirt_side_l_01_capsule, center=<3.6752 -1.4057 0>}]
 dyn_skirt_side_l_02 [{shapeType=Capsule, length=13.2998, radius=3, shapeName=dyn_skirt_side_l_02_capsule, center=<3.6752 -1.4057 0>}]
 dyn_skirt_side_l_03 [{shapeType=Capsule, length=13.2998, radius=3, shapeName=dyn_skirt_side_l_03_capsule, center=<3.6752 -1.4057 0>}]
 dyn_skirt_side_l_constr [{shapeType=Capsule, length=0.505, radius=0.505, shapeName=dyn_skirt_side_l_constr_capsule, rotation=<rot 0 -90 0>}]
 dyn_skirt_side_r_01 [{shapeType=Capsule, length=15.0518, radius=2, shapeName=dyn_skirt_side_r_01_capsule, center=<3.6752 -1.3455 0>}]
 dyn_skirt_side_r_02 [{shapeType=Capsule, length=15.0518, radius=3, shapeName=dyn_skirt_side_r_02_capsule, center=<3.6752 -1.3455 0>}]
 dyn_skirt_side_r_constr [{shapeType=Capsule, length=0.505, radius=0.505, shapeName=dyn_skirt_side_r_constr_capsule, rotation=<rot 0 -90 0>}]
 pelvis [{shapeType=Capsule, length=11.8668, radius=8.731, shapeName=pelvis_capsule, center=<-9.7115 -3.0082 0>}]
 thigh_l [{
   shapeType=Capsule
   length=38.8144
   radius=7.2705
   shapeName=thigh_l_capsule
   center=<-22.9109 -0.1638 0>
   rotation=<rot 90 179.8508 -178.7373>
  }]
 thigh_r [{shapeType=Capsule, length=38.4278, radius=7.4645, shapeName=thigh_l_capsule, center=<22.9109 0.1638 0>, rotation=<rot 90 179.8508 -178.7373>}]
}
constraints [
 {bone1Name=dyn_skirt_back_l_02, bone2Name=dyn_skirt_back_l_01, swing1LimitDegrees=35, swing2LimitDegrees=10, twistLimitDegrees=10}
 {bone1Name=dyn_skirt_back_l_03, bone2Name=dyn_skirt_back_l_02, swing1LimitDegrees=24, swing2LimitDegrees=10, twistLimitDegrees=10}
 {bone1Name=dyn_skirt_back_r_02, bone2Name=dyn_skirt_back_r_01, swing1LimitDegrees=35, swing2LimitDegrees=10, twistLimitDegrees=10}
 {bone1Name=dyn_skirt_front_l_02, bone2Name=dyn_skirt_front_l_01, swing1LimitDegrees=16, swing2LimitDegrees=10, twistLimitDegrees=10}
 {bone1Name=dyn_skirt_front_l_03, bone2Name=dyn_skirt_front_l_02, swing1LimitDegrees=16, swing2LimitDegrees=10, twistLimitDegrees=10}
 {bone1Name=dyn_skirt_front_r_02, bone2Name=dyn_skirt_front_r_01, swing1LimitDegrees=16, swing2LimitDegrees=10, twistLimitDegrees=10}
 {bone1Name=dyn_skirt_side_l_02, bone2Name=dyn_skirt_side_l_01, swing1LimitDegrees=10, swing2LimitDegrees=10, twistLimitDegrees=10}
 {bone1Name=dyn_skirt_side_l_03, bone2Name=dyn_skirt_side_l_02, swing1LimitDegrees=16, swing2LimitDegrees=10, twistLimitDegrees=10}
 {bone1Name=dyn_skirt_side_r_02, bone2Name=dyn_skirt_side_r_01, swing1LimitDegrees=16, swing2LimitDegrees=10, twistLimitDegrees=10}
 {bone1Name=dyn_skirt_front_l_01, bone2Name=dyn_skirt_front_l_constr, swing1LimitDegrees=10, swing2LimitDegrees=10, twistLimitDegrees=10}
 {bone1Name=dyn_skirt_side_l_01, bone2Name=dyn_skirt_side_l_constr, swing1LimitDegrees=10, swing2LimitDegrees=10, twistLimitDegrees=10}
 {bone1Name=dyn_skirt_back_l_01, bone2Name=dyn_skirt_back_l_constr, swing1LimitDegrees=16, swing2LimitDegrees=10, twistLimitDegrees=10}
 {bone1Name=dyn_skirt_back_r_01, bone2Name=dyn_skirt_back_r_constr, swing1LimitDegrees=16, swing2LimitDegrees=10, twistLimitDegrees=10}
 {bone1Name=dyn_skirt_side_r_01, bone2Name=dyn_skirt_side_r_constr, swing1LimitDegrees=10, swing2LimitDegrees=10, twistLimitDegrees=10}
 {bone1Name=dyn_skirt_front_r_01, bone2Name=dyn_skirt_front_r_constr, swing1LimitDegrees=10, swing2LimitDegrees=10, twistLimitDegrees=10}
 {
  bone1Name=thigh_l
  bone2Name=pelvis
  swing1LimitDegrees=45
  swing1Motion=Free
  swing2LimitDegrees=45
  swing2Motion=Free
  twistLimitDegrees=45
  twistMotion=Free
 }
 {
  bone1Name=thigh_r
  bone2Name=pelvis
  swing1LimitDegrees=45
  swing1Motion=Free
  swing2LimitDegrees=45
  swing2Motion=Free
  twistLimitDegrees=45
  twistMotion=Free
 }
 {
  bone1Name=calf_l
  bone2Name=thigh_l
  swing1LimitDegrees=45
  swing1Motion=Free
  swing2LimitDegrees=45
  swing2Motion=Free
  twistLimitDegrees=45
  twistMotion=Free
 }
 {
  bone1Name=calf_r
  bone2Name=thigh_r
  swing1LimitDegrees=45
  swing1Motion=Free
  swing2LimitDegrees=45
  swing2Motion=Free
  twistLimitDegrees=45
  twistMotion=Free
 }
 {
  bone1Name=dyn_skirt_back_l_constr
  bone2Name=pelvis
  swing1LimitDegrees=45
  swing1Motion=Free
  swing2LimitDegrees=45
  swing2Motion=Free
  twistLimitDegrees=45
  twistMotion=Free
 }
 {
  bone1Name=dyn_skirt_back_r_constr
  bone2Name=pelvis
  swing1LimitDegrees=45
  swing1Motion=Free
  swing2LimitDegrees=45
  swing2Motion=Free
  twistLimitDegrees=45
  twistMotion=Free
 }
 {
  bone1Name=dyn_skirt_front_l_constr
  bone2Name=pelvis
  swing1LimitDegrees=45
  swing1Motion=Free
  swing2LimitDegrees=45
  swing2Motion=Free
  twistLimitDegrees=45
  twistMotion=Free
 }
 {
  bone1Name=dyn_skirt_front_r_constr
  bone2Name=pelvis
  swing1LimitDegrees=45
  swing1Motion=Free
  swing2LimitDegrees=45
  swing2Motion=Free
  twistLimitDegrees=45
  twistMotion=Free
 }
 {
  bone1Name=dyn_skirt_side_l_constr
  bone2Name=pelvis
  swing1LimitDegrees=45
  swing1Motion=Free
  swing2LimitDegrees=45
  swing2Motion=Free
  twistLimitDegrees=45
  twistMotion=Free
 }
 {
  bone1Name=dyn_skirt_side_r_constr
  bone2Name=pelvis
  swing1LimitDegrees=45
  swing1Motion=Free
  swing2LimitDegrees=45
  swing2Motion=Free
  twistLimitDegrees=45
  twistMotion=Free
 }
]
