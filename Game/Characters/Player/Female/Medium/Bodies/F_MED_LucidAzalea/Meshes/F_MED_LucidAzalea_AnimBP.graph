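# Animation Blueprint for the LucidAzalea female player body, adding secondary physics and material-driven cloth motion.
# The parent mesh's pose is copied, run through a Control Rig, then a pelvis-space RigidBody node whose alpha drops slightly during skydiving.
# Each update feeds global and regional world position offset data to the body and overlay dynamic materials, switching cloth wind on while skydive floating.
# Those materials are cached when character customization loads.
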AnimBlueprint /Game/Characters/Player/Female/Medium/Bodies/F_MED_LucidAzalea/Meshes/F_MED_LucidAzalea_AnimBP
target_skeleton=&/F_MED_LucidAzalea_Skeleton
graphs [
 AnimGraph {nodes [
   n1 = Control Rig(Source=Copy Pose From Mesh() {bUseAttachedParent=true}) {
    alphaBoolBlend {blendOption=Linear}
    controlRigAssetReference {blueprintRigClass=&/F_MED_LucidAzalea_CtrlRig.F_MED_LucidAzalea_CtrlRig_C}
   }
   n2 = RigidBody(ComponentPose=Local To Component(LocalPose=@n1), Alpha=Get SkydiveAlpha() -> (SkydiveAlpha:double)) {
    bOverrideWorldGravity=true
    overridePhysicsAsset=&/F_MED_LucidAzalea_Physics
    simulationSpace=BaseBoneSpace
    alphaBoolBlend {blendOption=Linear}
    alphaScaleBiasClamp {bInterpResult=true}
    baseBoneRef {boneName=pelvis}
    componentAppliedLinearAccClamp=<1000 1000 1000>
    componentLinearAccScale=<0.7 0.7 0.7>
    componentLinearVelScale=<0.6 0.6 0.6>
    overrideWorldGravity {z=-980}
    simSpaceSettings {maxAngularAcceleration=250, maxLinearAcceleration=250, maxLinearVelocity=500, velocityScaleZ=0, worldAlpha=1}
   }
   Output Pose(Result=Component To Local(ComponentPose=@n2)) {layerGroup=None}
  ]}
 EventGraph {nodes [
   n1 = Event BlueprintUpdateAnimation()
   > n2 = Sequence()
   > Is Valid(InputObject=Get playerPawn() -> (playerPawn:Fort Player Pawn Athena Object Reference))
   > Set WPO Globals Function(
    Material=Get MatInstanceDynamic() -> (MatInstanceDynamic:Material Instance Dynamic Object Reference)
    Overlay Material=Get OverlayMID() -> (OverlayMID:Material Instance Dynamic Object Reference)
    Skeletal Mesh=Get SkelMesh()
    "Local Noise Transform  Location"="Get Local Noise Transform  Location"() -> ("Local Noise Transform  Location":Vector)
    Local Noise Offset=Get Local Noise Offset()
   )
   > Set WPO Regional Function!(
    execute
    Material:Material Instance Dynamic Object Reference=Get MatInstanceDynamic()
    Overlay Material:Material Instance Dynamic Object Reference=Get OverlayMID()
    Skeletal Mesh:Skeletal Mesh Component Object Reference=Get SkelMesh()
    Velocity Transform:Name=pelvis
    Velocity Interp Speed:double=10
    Collision Transform:Name=spine_01
    Collision Transform Axis:EAxis Enum
    Invert Collision Transform Axis:bool=false
    Delta Time X:double=@n1.DeltaTimeX
    Previous Velocity Transform Location:"Vector (by ref)"=Get Previous Velocity Transform Location() -> (Previous Velocity Transform Location:Vector)
    Previous Velocity Transform Velocity:"Vector (by ref)"=Get Previous Velocity Transform Velocity() -> (Previous Velocity Transform Velocity:Vector)
    Region:int=1
   ) -> (then)
   > n3 = Branch(Condition=Get bIsSkydiveFloating())
   > n4 = Set WPO Pho Wind Intensity(WPO Pho Wind Intensity=1)
   SetScalarParameterValue(
    execute=[@n5.then, @n4.then]
    self=[Get OverlayMID(), Get matInstanceDynamic() -> (matInstanceDynamic:Material Instance Dynamic Object Reference)]
    ParameterName=ClothSim_WindStrength
    Value=Get WPO Pho Wind Intensity()
   )
   n5 = Set WPO Pho Wind Intensity(execute=@n3.else, WPO Pho Wind Intensity=0)
   n6 = Branch(execute=@n2.then_1, Condition=Get bIsSkyDiving())
   > Set SkydiveAlpha(SkydiveAlpha:double=0.8) -> (Output_Get:double)
   Set SkydiveAlpha(execute=@n6.else, SkydiveAlpha=1)
   Event BlueprintInitializeAnimation()
   > Set SkelMesh(SkelMesh=GetOwnerSkeletalMeshComponent())
   n7 = TryGetPawnOwner()
   > Is Valid(InputObject=@n7)
   > n8 = Cast To FortPlayerPawnAthena(Object=@n7)
   > Set playerPawn(playerPawn:Fort Player Pawn Athena Object Reference=@n8) -> (Output_Get:Fort Player Pawn Athena Object Reference)
   Event OnCharacterCustomizationLoaded()
   > n9 = Set SkelMesh(SkelMesh=GetPartSkeletalMeshComponent())
   > n10 = Get Dynamic Material Instance(SkeletalMeshComponent=@n9)
   > Set MatInstanceDynamic(MatInstanceDynamic:Material Instance Dynamic Object Reference=@n10) -> (Output_Get:Material Instance Dynamic Object Reference)
   > n11 = GetOrCreateDynamicOverlayMaterial(MeshComponent=@n9, MaterialName=None)
   > Set OverlayMID(OverlayMID:Material Instance Dynamic Object Reference=@n11) -> (Output_Get:Material Instance Dynamic Object Reference)
  ]}
]
variables [
 Local Noise Offset {type=Vector}
 "Local Noise Transform  Location" {type=Vector}
 MatInstanceDynamic {type=MaterialInstanceDynamic}
 OverlayMID {type=MaterialInstanceDynamic}
 playerPawn {type=FortPlayerPawnAthena}
 Previous Velocity Transform Location {type=Vector}
 Previous Velocity Transform Velocity {type=Vector}
 SkelMesh {type=SkeletalMeshComponent}
 SkydiveAlpha {type=number}
 WPO Pho Wind Intensity {type=number}
]
